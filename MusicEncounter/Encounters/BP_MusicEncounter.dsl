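; BP_MusicEncounter runs one music encounter under a music encounter manager, from initialize through start, pause and fade-out on uninitialize.
; Lifecycle states are stored as bit flags, and added encounter components are tracked, optionally by gameplay tag.
; Spawn functions require a started encounter, then track each 2D, located or attached audio component and register it with the manager's registrar.
; Pause and phase reach the audio as parameters, with a handler setting phase from MetaSound output.
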(event Custom|Initialize%20Music%20Encounter
  (MusicEncounter|SetMusicEncounterState "NewEnumerator1" true))

(event Custom|Start%20Music%20Encounter
  (if (MusicEncounter|IsAbleToPlay)
    (MusicEncounter|SetMusicEncounterState "NewEnumerator2" true)
    (MusicEncounter|SetMusicEncounterState "NewEnumerator3")
    (for _array_element (Variables|MusicEncounter|GetMusic%20Audio%20Components)
      (Audio|Parameter|SetBooleanParameter _array_element "MusicEncounter.bIsPaused" false :!class /Script/AudioExtensions.AudioParameterControllerInterface))))

(event Custom|Pause%20Music%20Encounter
  (if (MusicEncounter|IsAbleToPause)
    (MusicEncounter|SetMusicEncounterState "NewEnumerator3" true)
    (MusicEncounter|SetMusicEncounterState "NewEnumerator2")
    (for _array_element (Variables|MusicEncounter|GetMusic%20Audio%20Components)
      (Audio|Parameter|SetBooleanParameter _array_element "MusicEncounter.bIsPaused" true :!class /Script/AudioExtensions.AudioParameterControllerInterface))))

(event Custom|Uninitialize%20Music%20Encounter
  (MusicEncounter|SetMusicEncounterState "NewEnumerator1")
  (MusicEncounter|SetMusicEncounterState "NewEnumerator2")
  (MusicEncounter|SetMusicEncounterState "NewEnumerator3")
  (for _array_element (Variables|MusicEncounter|GetMusic%20Audio%20Components)
    (Audio|Components|Audio|FadeOut _array_element (Variables|MusicEncounter|GetFade%20Out%20Duration) 0.0 "Linear" :!class /Script/Engine.AudioComponent)))

(fn AddMusicEncounterComponent
  (params (Music%20Encounter%20Component :type Music%20Encounter%20Component%20Base%20Object%20Reference%20%28by%20ref%29))
  (bind _music_encounter_component (Variables|GetMusic%20Encounter%20Component))
  (bind _component_tag (Variables|/MusicEncounter/Components/BP_MusicEncounterComponent.BP_MusicEncounterComponent_C|GetComponent%20Tag _music_encounter_component))
  (Utilities|IsValid Music%20Encounter%20Component
    (:Is%20Valid
      (Utilities|FlowControl|Sequence
        (:then_0
          (if (GameplayTags|IsGameplayTagValid _component_tag)
            (Utilities|Map|Add (Variables|MusicEncounter|GetTagged%20Music%20Encounter%20Components) _component_tag (Variables|GetMusic%20Encounter%20Component))))
        (:then_1
          (Utilities|FlowControl|Sequence
            (:then_0
              (Utilities|Array|AddUnique (Variables|MusicEncounter|GetMusic%20Encounter%20Components) (Variables|GetMusic%20Encounter%20Component)))
            (:then_1
              (Utilities|FlowControl|Sequence
                (:then_0
                  (MusicEncounter|CallOnMusicEncounterComponentAdded (Variables|GetMusic%20Encounter%20Component)))
                (:then_1
                  (Utilities|FlowControl|Sequence
                    (:then_0
                      (if (and (MusicEncounter|IsReadyToInitialize) (not (MusicEncounter|GetMusicEncounterStateValue "NewEnumerator1")))
                        (CallFunction|InitializeMusicEncounter)))))))))))))

(fn FindMusicEncounterComponentbyTag
  (params (Component%20Tag :type Gameplay%20Tag%20Structure))
  (return (Utilities|Map|Find (Variables|MusicEncounter|GetTagged%20Music%20Encounter%20Components) Component%20Tag)))

(fn GetMusicAudioComponents
  (return (Variables|MusicEncounter|GetMusic%20Audio%20Components)))

(fn GetMusicEncounterComponents
  (return (Variables|MusicEncounter|GetMusic%20Encounter%20Components)))

(fn GetMusicEncounterInstanceTag
  (return (Variables|MusicEncounter|GetMusic%20Encounter%20Instance%20Tag)))

(fn GetMusicEncounterState
  (return (Variables|MusicEncounter|GetMusic%20Encounter%20State)))

(fn GetMusicEncounterStateValue
  (params (State%20Type :type EMusicEncounterState%20Enum :default "NewEnumerator1"))
  (return (Math|Conversions|ToBoolean(Integer) (Math|Integer|BitwiseAND (Variables|MusicEncounter|GetMusic%20Encounter%20State) (Math|Float|Round (Math|Float|Power 2.0 (Math|Conversions|ToFloat(Byte) State%20Type)))))))

(fn GetMusicEventTag
  (params (Audio%20Component :type Audio%20Component%20Object%20Reference))
  (return (Variables|MusicEncounter|MusicEventSubsystem|GetMusic%20Event%20Tag)))

(fn GetMusicGroupName
  (return (Variables|MusicEncounter|MusicEventSubsystem|GetMusic%20Group%20Name)))

(fn GetMusicManager
  (bind _asmusic_encounter_manager_component (Utilities|Casting|CastToBP_MusicEncounterManager (Utilities|GetOuterObject self)))
  (return _asmusic_encounter_manager_component))

(fn HandlePhaseChanged
  (params (OutputName :type Name) (Output :type Meta%20Sound%20Output%20Structure%20%28by%20ref%29))
  (bind (_success _returnvalue) (MetaSoundOutput|GetInt32 Output))
  (MusicEncounter|SetMusicEncounterPhase _returnvalue))

(fn IsAbletoPause
  (return (MusicEncounter|GetMusicEncounterStateValue "NewEnumerator2")))

(fn IsAbletoPlay
  (return (and (MusicEncounter|GetMusicEncounterStateValue "NewEnumerator1") (or (not (Replay|GetIsPlayingReplay (Utilities|GetOuterObject self))) (not (Variables|MusicEncounter|GetDisable%20During%20Replays))))))

(fn IsReadytoInitialize
  (return true))

(fn IsValid
  (return true))

(fn RegisterMusicbyComponent
  (params (Audio%20Component :type Audio%20Component%20Object%20Reference))
  (bind _music_manager_component (MusicEncounter|GetMusicManager))
  (bind _music_registrar (MusicEncounterManager|GetOrCreateMusicRegistrar _music_manager_component))
  (Utilities|IsValid Audio%20Component
    (:Is%20Valid
      (Utilities|IsValid _music_manager_component
        (:Is%20Valid
          (Utilities|IsValid _music_registrar
            (:Is%20Valid
              (MusicRegistration|RegisterMusicByComponent _music_registrar self (Variables|GetAudio%20Component)))))))))

(fn RemoveFinishedAudioComponents
  (for _array_element (Variables|MusicEncounter|GetMusic%20Audio%20Components)
    (bind _output_get (LocalVariables|SetAudio%20Component _array_element))
    (Utilities|IsValid _output_get
      (:Is%20Valid
        (if (not (Audio|Components|Audio|IsPlaying (LocalVariables|GetAudio%20Component)))
          (bind _returnvalue (Utilities|Array|RemoveItem (Variables|MusicEncounter|GetMusic%20Audio%20Components) (LocalVariables|GetAudio%20Component)))
          (if _returnvalue
            (MusicEncounter|CallOnMusicAudioComponentRemoved (LocalVariables|GetAudio%20Component))))))))

(fn RemoveMusicEncounterComponent
  (params (Music%20Encounter%20Component :type Music%20Encounter%20Component%20Base%20Object%20Reference%20%28by%20ref%29))
  (bind _music_encounter_component (Variables|GetMusic%20Encounter%20Component))
  (bind _component_tag (Variables|/MusicEncounter/Components/BP_MusicEncounterComponent.BP_MusicEncounterComponent_C|GetComponent%20Tag _music_encounter_component))
  (Utilities|IsValid Music%20Encounter%20Component
    (:Is%20Valid
      (bind _returnvalue (Utilities|Array|RemoveItem (Variables|MusicEncounter|GetMusic%20Encounter%20Components) (Variables|GetMusic%20Encounter%20Component)))
      (if _returnvalue
        (Utilities|FlowControl|Sequence
          (:then_0
            (if (GameplayTags|IsGameplayTagValid _component_tag)
              (Utilities|Map|Remove (Variables|MusicEncounter|GetTagged%20Music%20Encounter%20Components) _component_tag)))
          (:then_1
            (Utilities|FlowControl|Sequence
              (:then_0
                (MusicEncounter|CallOnMusicEncounterComponentRemoved (Variables|GetMusic%20Encounter%20Component)))
              (:then_1
                (Utilities|FlowControl|Sequence
                  (:then_0
                    (if (not (MusicEncounter|IsValid))
                      (CallFunction|UninitializeMusicEncounter))))))))))))

(fn SetInstanceID
  (params (Encounter%20ID :type Guid%20Structure))
  (Variables|MusicEncounter|SetMusic%20Encounter%20Instance%20ID Encounter%20ID))

(fn SetInstanceTag
  (params (Music%20Encounter%20Instance%20Tag :type Gameplay%20Tag%20Structure))
  (Variables|MusicEncounter|SetMusic%20Encounter%20Instance%20Tag Music%20Encounter%20Instance%20Tag))

(fn SetMusicEncounterPhase
  (params (Phase :type Integer))
  (Variables|MusicEncounter|SetMusic%20Encounter%20Phase Phase)
  (for _array_element (Variables|MusicEncounter|GetMusic%20Audio%20Components)
    (Audio|Parameter|SetIntegerParameter _array_element "MusicEncounter.Phase" (Variables|MusicEncounter|GetMusic%20Encounter%20Phase) :!class /Script/AudioExtensions.AudioParameterControllerInterface)))

(fn SetMusicEncounterState
  (params (Music%20Encounter%20State :type EMusicEncounterState%20Enum :default "NewEnumerator1") (Value :type Boolean))
  (if (!= (MusicEncounter|GetMusicEncounterStateValue Music%20Encounter%20State) Value)
    (Variables|MusicEncounter|SetMusic%20Encounter%20State (Math|Integer|BitwiseXOR (Variables|MusicEncounter|GetMusic%20Encounter%20State) (Math|Float|Round (Math|Float|Power 2.0 (Math|Conversions|ToFloat(Byte) (Variables|MusicEncounter|GetMusic%20Encounter%20State))))))
    (MusicEncounter|CallOnMusicEncounterStateChanged (Variables|MusicEncounter|GetMusic%20Encounter%20State))))

(fn SetUpNewAudioComponent
  (params (Audio :type Audio%20Component%20Object%20Reference%20%28by%20ref%29))
  (Utilities|IsValid Audio
    (:Is%20Valid
      (Utilities|FlowControl|Sequence
        (:then_0
          (Utilities|Array|AddUnique (Variables|MusicEncounter|GetMusic%20Audio%20Components) (Variables|GetAudio)))
        (:then_1
          (Utilities|FlowControl|Sequence
            (:then_0
              (MusicEncounter|CallOnMusicAudioComponentAdded (Variables|GetAudio))
              (EventDispatchers|BindEventtoOnAudioFinished (Variables|GetAudio) (EventDispatchers|CreateEvent :!event Remove%20Finished%20Audio%20Components)))
            (:then_1
              (Utilities|FlowControl|Sequence
                (:then_0
                  (MusicEncounter|RegisterMusicByComponent (Variables|GetAudio)))))))))))

(fn SpawnMusic2D
  (params (Music :type Sound%20Base%20Object%20Reference) (Attach%20Component :type Scene%20Component%20Object%20Reference) (Owning%20Actor :type Actor%20Object%20Reference))
  (if (MusicEncounter|GetMusicEncounterStateValue "NewEnumerator2")
    (bind _output_get (LocalVariables|SetOwner (select (Utilities|IsValid (Variables|GetOwning%20Actor) :!class /Script/Engine.KismetSystemLibrary) (Variables|GetOwning%20Actor) (Utilities|GetOuterObject self))))
    (bind _returnvalue (Audio|SpawnSound2D _output_get (Variables|GetMusic)))
    (bind _output_get_2 (LocalVariables|SetAudio _returnvalue))
    (Utilities|IsValid _output_get_2
      (:Is%20Valid
        (MusicEncounter|SetUpNewAudioComponent (LocalVariables|GetAudio))
        (Utilities|FlowControl|Sequence
          (:then_0
            (bind _attach_component (Variables|GetAttach%20Component)
              (:then
                (Transformation|AttachComponentToComponent (LocalVariables|GetAudio) _attach_component "None" "SnapToTarget" "SnapToTarget" "SnapToTarget" true))
              (:else)))
          (:then_1
            (Utilities|FlowControl|Sequence
              (:then_0
                (return (LocalVariables|GetAudio))))))))
    (else
      (Utilities|Debugging|RaiseScriptError (Utilities|String|Append "Could not begin playing music '" (Utilities|GetDisplayName (Variables|GetMusic)) "'. Start Music Encounter has not been called.")))))

(fn SpawnMusicatLocation
  (params (Music :type Sound%20Base%20Object%20Reference) (Location :type Vector) (On%20Virtualization%20Changed :type Delegate%20%28by%20ref%29))
  (if (MusicEncounter|GetMusicEncounterStateValue "NewEnumerator2")
    (bind _returnvalue (Audio|SpawnSoundatLocation (Utilities|GetOuterObject self) (Variables|GetMusic) (Variables|GetLocation)))
    (bind _output_get (LocalVariables|SetAudio _returnvalue))
    (Utilities|IsValid _output_get
      (:Is%20Valid
        (MusicEncounter|SetUpNewAudioComponent (LocalVariables|GetAudio))
        (EventDispatchers|BindEventtoOnAudioVirtualizationChanged (LocalVariables|GetAudio) (Variables|GetOn%20Virtualization%20Changed))
        (return (LocalVariables|GetAudio))))
    (else
      (Utilities|Debugging|RaiseScriptError (Utilities|String|Append "Could not begin playing music '" (Utilities|GetDisplayName (Variables|GetMusic)) "'. Start Music Encounter has not been called.")))))

(fn SpawnMusicAttached
  (params (Music :type Sound%20Base%20Object%20Reference) (Attach%20to%20Component :type Scene%20Component%20Object%20Reference) (Location%20Offset :type Vector) (Stop%20When%20Attached%20To%20Destroyed :type Boolean :default true) (On%20Virtualization%20Changed :type Delegate%20%28by%20ref%29))
  (if (MusicEncounter|GetMusicEncounterStateValue "NewEnumerator2")
    (bind _returnvalue (Audio|SpawnSoundAttached :Sound (Variables|GetMusic) :AttachToComponent (Variables|GetAttach%20to%20Component) :AttachPointName "None" :Location (Variables|GetLocation%20Offset) :LocationType "KeepRelativeOffset" :bStopWhenAttachedToDestroyed (Variables|GetStop%20When%20Attached%20To%20Destroyed) :VolumeMultiplier 1.0 :PitchMultiplier 1.0 :StartTime 0.0 :bAutoDestroy true))
    (bind _output_get (LocalVariables|SetAudio _returnvalue))
    (Utilities|IsValid _output_get
      (:Is%20Valid
        (MusicEncounter|SetUpNewAudioComponent (LocalVariables|GetAudio))
        (EventDispatchers|BindEventtoOnAudioVirtualizationChanged (LocalVariables|GetAudio) (Variables|GetOn%20Virtualization%20Changed))
        (return (LocalVariables|GetAudio))))
    (else
      (Utilities|Debugging|RaiseScriptError (Utilities|String|Append "Could not begin playing music '" (Utilities|GetDisplayName (Variables|GetMusic)) "'. Start Music Encounter has not been called.")))))
