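; Boss music encounter Blueprint that plays idle and combat music on a tagged boss actor and winds down when the boss dies.
; Idle, Combat and Completed states spawn and fade 2D tracks, driven by death and optional alert-level events.
; Phase and alert level feed audio parameters, while a delegates object broadcasts state, phase and targeting changes.
; The listener volume starts and pauses music, which keeps playing outside it while the boss targets the listener.
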(event Custom|Enter%20Idle%20State
  (if (MusicEncounter|GetMusicEncounterStateValue "NewEnumerator2")
    (Variables|MusicEncounter|Boss|Idle|GetIdle%20Music%20Audio%20Component
      (:then)
      (:else
        (bind _boss_actor (Variables|MusicEncounter|Boss|GetBoss%20Actor)
          (:then
            (bind _audio_component (MusicEncounter|SpawnMusic2D (Variables|MusicEncounter|Boss|Idle|GetIdle%20Music) (Variables|/Script/Engine.Actor|GetRootComponent _boss_actor) _boss_actor))
            (Variables|MusicEncounter|Boss|Idle|SetIdle%20Music%20Audio%20Component _audio_component)
            (bind _combat_music_audio_component (Variables|MusicEncounter|Boss|Combat|GetCombat%20Music%20Audio%20Component)
              (:then
                (Audio|Components|Audio|FadeOut _combat_music_audio_component (Variables|MusicEncounter|Boss|Combat|GetCombat%20Music%20Fade%20Out%20Duration) 0.0 "Linear" :!class /Script/Engine.AudioComponent))
              (:else)))
          (:else))))))

(event EventInitializeMusicEncounter
  (|CallParentFunction)
  (Game|BindEventtoOnEndPlay (Variables|MusicEncounter|Boss|GetBoss%20Actor) (EventDispatchers|CreateEvent :!event Custom|On%20Boss%20Actor%20End%20Play))
  (bind _returnvalue (Game|ConstructObjectfromClass "/MusicEncounter/DelegateObjects/BP_MusicEncounterDelegates_Boss.BP_MusicEncounterDelegates_Boss_C"))
  (Variables|MusicEncounter|Boss|SetBoss%20Delegates _returnvalue))

(event Custom|Enter%20Combat%20State
  (if (MusicEncounter|GetMusicEncounterStateValue "NewEnumerator2")
    (Variables|MusicEncounter|Boss|Combat|GetCombat%20Music%20Audio%20Component
      (:then)
      (:else
        (bind _boss_actor (Variables|MusicEncounter|Boss|GetBoss%20Actor)
          (:then
            (bind _audio_component (MusicEncounter|SpawnMusic2D (Variables|MusicEncounter|Boss|Combat|GetCombat%20Music) (Variables|/Script/Engine.Actor|GetRootComponent _boss_actor) _boss_actor))
            (Variables|MusicEncounter|Boss|Combat|SetCombat%20Music%20Audio%20Component _audio_component)
            (bind _idle_music_audio_component (Variables|MusicEncounter|Boss|Idle|GetIdle%20Music%20Audio%20Component)
              (:then
                (Audio|Components|Audio|FadeOut _idle_music_audio_component (Variables|MusicEncounter|Boss|Idle|GetIdle%20Music%20Fade%20Out%20Duration) 0.0 "Linear" :!class /Script/Engine.AudioComponent))
              (:else)))
          (:else))))))

(event EventUninitializeMusicEncounter
  (|CallParentFunction)
  (Variables|MusicEncounter|Boss|SetBoss%20Delegates 0)
  (bind _boss_actor (Variables|MusicEncounter|Boss|GetBoss%20Actor)
    (:then
      (Game|UnbindEventfromOnEndPlay _boss_actor (EventDispatchers|CreateEvent :!event Custom|On%20Boss%20Actor%20End%20Play)))
    (:else)))

(event Custom|Enter%20Completed%20State
  (bind _ending_phase_index (Variables|MusicEncounter|Boss|Ending|GetEnding%20Phase%20Index))
  (Utilities|FlowControl|Sequence
    (:then_0
      (if (and (not (Utilities|IsValid (Variables|MusicEncounter|Boss|Combat|GetCombat%20Music%20Audio%20Component) :!class /Script/Engine.KismetSystemLibrary)) (not (Utilities|IsValid (Variables|MusicEncounter|Boss|Idle|GetIdle%20Music%20Audio%20Component) :!class /Script/Engine.KismetSystemLibrary)))
        (CallFunction|UninitializeMusicEncounter)))
    (:then_1
      (Utilities|FlowControl|Sequence
        (:then_0
          (bind _combat_music_audio_component (Variables|MusicEncounter|Boss|Combat|GetCombat%20Music%20Audio%20Component)
            (:then
              (EventDispatchers|BindEventtoOnAudioFinished _combat_music_audio_component (EventDispatchers|CreateEvent :!event Uninitialize%20Music%20Encounter)))
            (:else)))
        (:then_1
          (Utilities|FlowControl|Sequence
            (:then_0
              (bind _idle_music_audio_component (Variables|MusicEncounter|Boss|Idle|GetIdle%20Music%20Audio%20Component)
                (:then
                  (Audio|Components|Audio|FadeOut _idle_music_audio_component (Variables|MusicEncounter|Boss|Idle|GetIdle%20Music%20Fade%20Out%20Duration) 0.0 "Linear" :!class /Script/Engine.AudioComponent)
                  (EventDispatchers|BindEventtoOnAudioFinished (Variables|MusicEncounter|Boss|Idle|GetIdle%20Music%20Audio%20Component) (EventDispatchers|CreateEvent :!event Uninitialize%20Music%20Encounter)))
                (:else)))
            (:then_1
              (Utilities|FlowControl|Sequence
                (:then_0
                  (MusicEncounter|SetMusicEncounterPhase _ending_phase_index)
                  (CallFunction|SetBossEncounterCombatPhase _ending_phase_index))))))))))

(event EventStartMusicEncounter
  (|CallParentFunction)
  (CallFunction|SetBossEncounterState (Variables|MusicEncounter|Boss|GetBoss%20Encounter%20State)))

(event MusicEncounter|Volume|EventListenerEntered
  (|CallParentFunction)
  (CallFunction|StartMusicEncounter))

(event MusicEncounter|Volume|EventListenerExited
  (|CallParentFunction)
  (Variables|MusicEncounter|Boss|GetIs%20Listener%20Targeted
    (:then)
    (:else
      (CallFunction|PauseMusicEncounter))))

(event Custom|Set%20Boss%20Encounter%20State
  (params (Boss%20Encounter%20State :type EMusicEncounterBossState%20Enum :default "NewEnumerator7"))
  (bind _output_get (Variables|MusicEncounter|Boss|SetBoss%20Encounter%20State Boss%20Encounter%20State))
  (Utilities|FlowControl|Sequence
    (:then_0
      (switch Utilities|FlowControl|Switch|SwitchonEMusicEncounterBossState _output_get
        (:NewEnumerator5
          (CallFunction|EnterIdleState))
        (:NewEnumerator2
          (CallFunction|EnterCombatState))
        (:NewEnumerator4
          (CallFunction|EnterCompletedState))))
    (:then_1
      (bind _boss_delegates (Variables|MusicEncounter|Boss|GetBoss%20Delegates)
        (:then
          (MusicEncounter|Boss|CallOnBossEncounterStateChanged _boss_delegates (Variables|MusicEncounter|Boss|GetBoss%20Encounter%20State)))
        (:else)))))

(event Custom|Set%20Boss%20Encounter%20Combat%20Phase
  (params (Boss%20Encounter%20Combat%20Phase :type Integer))
  (Variables|MusicEncounter|Boss|SetBoss%20Encounter%20Combat%20Phase Boss%20Encounter%20Combat%20Phase)
  (Utilities|FlowControl|Sequence
    (:then_0
      (for _array_element (MusicEncounter|GetMusicAudioComponents)
        (Audio|Parameter|SetIntegerParameter _array_element "MusicEncounter.Boss.Phase" Boss%20Encounter%20Combat%20Phase :!class /Script/AudioExtensions.AudioParameterControllerInterface)))
    (:then_1
      (bind _boss_delegates (Variables|MusicEncounter|Boss|GetBoss%20Delegates)
        (:then
          (MusicEncounter|Boss|CallOnBossEncounterCombatPhaseChanged _boss_delegates Boss%20Encounter%20Combat%20Phase))
        (:else)))))

(event MusicEncounter|Boss|EventBossDeath
  (params (Boss%20Component%20Interface :type Music%20Encounter%20Component%20Boss%20Interface%20Interface))
  (CallFunction|SetBossEncounterState "NewEnumerator4"))

(event MusicEncounter|Boss|EventBossAlertLevelChanged
  (params (Boss%20Component%20Interface :type Music%20Encounter%20Component%20Boss%20Interface%20Interface) (Alert%20Level :type EAlertLevel%20Enum :default "Unaware"))
  (Variables|MusicEncounter|Boss|GetUse%20Alertness%20to%20Toggle%20Boss%20State
    (:then
      (Utilities|FlowControl|Sequence
        (:then_0
          (if (Utilities|Enum|NotEqual(Enum) (Variables|MusicEncounter|Boss|GetBoss%20Encounter%20State) "NewEnumerator4")
            (switch Utilities|FlowControl|Switch|SwitchonEAlertLevel Alert%20Level
              (:Unaware
                (CallFunction|SetBossEncounterState "NewEnumerator5"))
              (:Threatened
                (CallFunction|SetBossEncounterState "NewEnumerator2")))))
        (:then_1
          (for _array_element (MusicEncounter|GetMusicAudioComponents)
            (Audio|Parameter|SetIntegerParameter _array_element "AlertLevel" (Utilities|Select :Index Alert%20Level :Unaware 0 :Suspicious 1 :LostTarget 1 :Threatened 2 :Count 0) :!class /Script/AudioExtensions.AudioParameterControllerInterface)))))
    (:else)))

(event MusicEncounter|Boss|EventIsBossTargetingListenerChanged
  (params (Boss%20Component%20Interface :type Music%20Encounter%20Component%20Boss%20Interface%20Interface) (Is%20Boss%20Targeting%20Listener :type Boolean))
  ;[1 1 1] Allows music to continue playing if the player is targeted by the boss, even if the player leaves the encounter volume. Boss must have `Uses Target Changed Replication` enabled
  (bind _returnvalue (MusicEncounterComponent|Boss|UsesTargetChangedReplication Boss%20Component%20Interface :!class /MusicEncounter/Interfaces/Components/BPI_MusicEncounterComponent_Boss.BPI_MusicEncounterComponent_Boss_C))
  (if _returnvalue
    (Variables|MusicEncounter|Boss|SetIs%20Listener%20Targeted Is%20Boss%20Targeting%20Listener)
    (Utilities|FlowControl|Sequence
      (:then_0
        (Variables|MusicEncounter|Boss|GetIs%20Listener%20Targeted
          (:then
            (CallFunction|StartMusicEncounter))
          (:else
            (if (not (MusicEncounter|Volume|IsListenerInVolume))
              (CallFunction|PauseMusicEncounter)))))
      (:then_1
        (bind _boss_delegates (Variables|MusicEncounter|Boss|GetBoss%20Delegates)
          (:then
            (MusicEncounter|Boss|CallOnIsListenerTargetedChanged _boss_delegates (Variables|MusicEncounter|Boss|GetIs%20Listener%20Targeted)))
          (:else))))))

(event Custom|On%20Boss%20Actor%20End%20Play
  (params (Actor :type Actor%20Object%20Reference) (EndPlayReason :type EEndPlayReason%20Enum :default "Destroyed"))
  (if (Utilities|Enum|NotEqual(Enum) (Variables|MusicEncounter|Boss|GetBoss%20Encounter%20State) "NewEnumerator4")
    (CallFunction|UninitializeMusicEncounter)))

(fn AddMusicEncounterComponent
  (params (Music%20Encounter%20Component :type Music%20Encounter%20Component%20Base%20Object%20Reference%20%28by%20ref%29))
  (Utilities|IsValid Music%20Encounter%20Component
    (:Is%20Valid
      (if (== (Variables|/MusicEncounter/Components/BP_MusicEncounterComponent.BP_MusicEncounterComponent_C|GetComponent%20Tag (Variables|GetMusic%20Encounter%20Component)) (Variables|MusicEncounter|Boss|GetBoss%20Component%20Tag))
        (Variables|MusicEncounter|Boss|SetBoss%20Actor (Components|GetOwner (Variables|GetMusic%20Encounter%20Component)))
        (|CallParentFunction (Variables|GetMusic%20Encounter%20Component))
        (return)
        (else
          _)))
    (:Is%20Not%20Valid
      _))
  (|CallParentFunction (Variables|GetMusic%20Encounter%20Component))
  (return))

(fn GetBossEncounterCombatPhase
  (return (Variables|MusicEncounter|Boss|GetBoss%20Encounter%20Combat%20Phase)))

(fn GetBossEncounterDelegates
  (return (Variables|MusicEncounter|Boss|GetBoss%20Delegates)))

(fn GetBossEncounterState
  (return (Variables|MusicEncounter|Boss|GetBoss%20Encounter%20State)))

(fn IsReadytoInitialize
  (return (Utilities|IsValid (Variables|MusicEncounter|Boss|GetBoss%20Actor) :!class /Script/Engine.KismetSystemLibrary)))
