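; Music encounter Blueprint that gives an encounter a runtime-spawned Audio Gameplay Volume (AGV) with adjustable radius and half height.
; On initialization, it optionally spawns the AGV through the music manager, sizes it via message calls and attaches it by component tag.
; Attachment snaps the AGV to the root of a given actor or of the tagged encounter component's owner, keeping relative scale.
; Uninitialization destroys the AGV.
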(event EventInitializeMusicEncounter
  (|CallParentFunction)
  (Variables|MusicEncounter|Volume|GetSpawn%20AGV%20on%20Initialization
    (:then
      (bind _music_encounter_agv (MusicEncounterManager|SpawnAudioGameplayVolume (MusicEncounter|GetMusicManager) (Variables|MusicEncounter|Volume|GetSpawn%20AGV%20Class) self))
      (Variables|MusicEncounter|Volume|SetMusic%20Encounter%20AGV _music_encounter_agv)
      (MusicEncounter|Volume|SetVolumeRadius (Variables|MusicEncounter|Volume|GetRadius))
      (MusicEncounter|Volume|SetVolumeHalfHeight (Variables|MusicEncounter|Volume|GetHalf%20Height))
      (MusicEncounter|Volume|AttachAGVToActorByComponentTag (Variables|MusicEncounter|Volume|GetAGV%20Attach%20Component%20Tag)))
    (:else)))

(event EventUninitializeMusicEncounter
  (|CallParentFunction)
  (bind _music_encounter_agv (Variables|MusicEncounter|Volume|GetMusic%20Encounter%20AGV)
    (:then
      (Actor|DestroyActor _music_encounter_agv))
    (:else)))

(fn AttachAGVtoActor
  (params (Actor :type Actor%20Object%20Reference))
  (Utilities|IsValid Actor
    (:Is%20Valid
      (bind _music_encounter_agv (Variables|MusicEncounter|Volume|GetMusic%20Encounter%20AGV)
        (:then
          (Transformation|AttachActorToComponent _music_encounter_agv (Variables|/Script/Engine.Actor|GetRootComponent Actor) "None" "SnapToTarget" "SnapToTarget" "KeepRelative" true))
        (:else)))))

(fn AttachAGVtoActorbyComponentTag
  (params (Component%20Tag :type Gameplay%20Tag%20Structure))
  (bind _music_encounter_component (MusicEncounter|FindMusicEncounterComponentByTag Component%20Tag))
  (bind _returnvalue (Components|GetOwner _music_encounter_component))
  (Utilities|IsValid _music_encounter_component
    (:Is%20Valid
      (Utilities|IsValid _returnvalue
        (:Is%20Valid
          (bind _music_encounter_agv (Variables|MusicEncounter|Volume|GetMusic%20Encounter%20AGV)
            (:then
              (Transformation|AttachActorToComponent _music_encounter_agv (Variables|/Script/Engine.Actor|GetRootComponent _returnvalue) "None" "SnapToTarget" "SnapToTarget" "KeepRelative" true))
            (:else)))))))

(fn GetMusicEncounterAGV
  (return (Variables|MusicEncounter|Volume|GetMusic%20Encounter%20AGV)))

(fn SetVolumeHalfHeight
  (params (Half%20Height :type Float%20%28double-precision%29))
  (bind _output_get (Variables|MusicEncounter|Volume|SetHalf%20Height Half%20Height))
  (MusicEncounter|Volume|SetCylinderHalfHeight(Message) (MusicEncounter|Volume|GetMusicEncounterAGV) _output_get))

(fn SetVolumeRadius
  (params (Radius :type Float%20%28double-precision%29))
  (bind _output_get (Variables|MusicEncounter|Volume|SetRadius Radius))
  (MusicEncounter|Volume|SetRadius(Message) (MusicEncounter|Volume|GetMusicEncounterAGV) _output_get))
