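; Music encounter that plays objective music while a listener is in its volume and ends it once the objective completes.
; The objective component is picked by tag as components register, and the encounter is valid only with one.
; Start spawns 2D music once, passing the objective owner's root; leaving pauses the encounter.
; An async task awaits the owner's completion gameplay tag, fires an audio trigger parameter, then uninitializes when audio finishes.
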(event EventInitializeMusicEncounter
  (|CallParentFunction)
  (bind _asynctaskproxy (Ability|Tasks|WaitGameplayTagAddToActor (Components|GetOwner (Variables|MusicEncounter|Objective|GetObjective%20Component)) (Variables|MusicEncounter|Objective|GetObjective%20Complete%20Tag) true)
    (:then
      (Variables|MusicEncounter|Objective|SetWait%20Objective%20Complete%20Task _asynctaskproxy))
    (:Added
      (CallFunction|ObjectiveCompleted))))

(event EventStartMusicEncounter
  (|CallParentFunction)
  (Variables|MusicEncounter|Objective|GetObjective%20Music%20Audio%20Component
    (:then)
    (:else
      (bind _objective_component (Variables|MusicEncounter|Objective|GetObjective%20Component)
        (:then
          (Utilities|IsValid (Components|GetOwner _objective_component)
            (:Is%20Valid
              (bind _audio_component (MusicEncounter|SpawnMusic2D (Variables|MusicEncounter|Objective|GetObjective%20Music) (Variables|/Script/Engine.Actor|GetRootComponent (Components|GetOwner _objective_component))))
              (Variables|MusicEncounter|Objective|SetObjective%20Music%20Audio%20Component _audio_component))))
        (:else)))))

(event EventUninitializeMusicEncounter
  (|CallParentFunction)
  (bind _wait_objective_complete_task (Variables|MusicEncounter|Objective|GetWait%20Objective%20Complete%20Task)
    (:then
      (AsyncAction|Cancel _wait_objective_complete_task))
    (:else)))

(event MusicEncounter|Volume|EventListenerEntered
  (|CallParentFunction)
  (CallFunction|StartMusicEncounter))

(event MusicEncounter|Volume|EventListenerExited
  (|CallParentFunction)
  (CallFunction|PauseMusicEncounter))

(event Custom|Objective%20Completed
  (bind _objective_music_audio_component (Variables|MusicEncounter|Objective|GetObjective%20Music%20Audio%20Component)
    (:then
      (Audio|Parameter|ExecuteTriggerParameter _objective_music_audio_component "MusicEncounter.Objective.Complete")
      (EventDispatchers|BindEventtoOnAudioFinished (Variables|MusicEncounter|Objective|GetObjective%20Music%20Audio%20Component) (EventDispatchers|CreateEvent :!event Uninitialize%20Music%20Encounter)))
    (:else)))

(fn AddMusicEncounterComponent
  (params (Music%20Encounter%20Component :type Music%20Encounter%20Component%20Base%20Object%20Reference%20%28by%20ref%29))
  (Utilities|IsValid Music%20Encounter%20Component
    (:Is%20Valid
      (if (== (Variables|/MusicEncounter/Components/BP_MusicEncounterComponent.BP_MusicEncounterComponent_C|GetComponent%20Tag (Variables|GetMusic%20Encounter%20Component)) (Variables|MusicEncounter|Objective|GetObjective%20Component%20Tag))
        (bind _output_get (Variables|MusicEncounter|Objective|SetObjective%20Component (Variables|GetMusic%20Encounter%20Component)))
        (|CallParentFunction _output_get)
        (return)
        (else
          _)))
    (:Is%20Not%20Valid
      _))
  (|CallParentFunction (Variables|GetMusic%20Encounter%20Component))
  (return))

(fn IsValid
  (return (Utilities|IsValid (Variables|MusicEncounter|Objective|GetObjective%20Component) :!class /Script/Engine.KismetSystemLibrary)))
