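; Music encounter that plays diegetic objective music attached to the actor owning a tagged encounter component.
; AddMusicEncounterComponent captures the component matching the attach tag; once game state is ready, initialization listens for the stateful ActiveState event and starts the encounter.
; Start spawns the music once on the owner's root at a relative offset.
; Clearing that event or the owner's EndPlay fades the audio out, with uninitialization bound to audio finishing.
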(event EventInitializeMusicEncounter
  (|CallParentFunction)
  (bind _objective_component (Variables|MusicEncounter|Objective|GetObjective%20Component)
    (:then
      (Utilities|IsValid (Components|GetOwner _objective_component)
        (:Is%20Valid
          (Game|BindEventtoOnEndPlay (Components|GetOwner _objective_component) (EventDispatchers|CreateEvent :!event Custom|On%20Owner%20End%20Play))
          (Class|FortAsyncActionGameStateReady|GameStateReadyAsync (Components|GetOwner _objective_component)
            (:OnReady
              (bind _returnvalue (GameplayEventRouter|GetEventRouter "/Script/FortniteGame.FortGameState" (Components|GetOwner _objective_component)))
              (bind _asynctaskproxy (GameplayEventRouter|StartListeningToStatefulEvent _returnvalue _returnvalue "/MusicEncounter/Events/GameplayEvent_MusicEncounter_ActiveState.GameplayEvent_MusicEncounter_ActiveState")
                (:then
                  (Variables|MusicEncounter|Objective|SetObjective%20Event%20State%20Task _asynctaskproxy))
                (:OnEventReceived
                  _)
                (:OnSavedEventStateExists
                  _)
                (:OnEventStateCleared
                  (CallFunction|StopMusic)
                  (GameplayEventRouter|StopListeningtoStatefulEvent (Variables|MusicEncounter|Objective|GetObjective%20Event%20State%20Task))))
              (CallFunction|StartMusicEncounter))))))
    (:else)))

(event EventStartMusicEncounter
  (|CallParentFunction)
  (Variables|MusicEncounter|Objective|GetObjective%20Music%20Audio%20Component
    (:then)
    (:else
      (bind _objective_component (Variables|MusicEncounter|Objective|GetObjective%20Component)
        (:then
          (Utilities|IsValid _objective_component
            (:Is%20Valid
              (Utilities|IsValid (Components|GetOwner _objective_component)
                (:Is%20Valid
                  (bind _audio_component (MusicEncounter|SpawnMusicAttached (Variables|MusicEncounter|Objective|GetObjective%20Music) (Variables|/Script/Engine.Actor|GetRootComponent (Components|GetOwner _objective_component)) (Variables|MusicEncounter|Objective|GetRelative%20Location%20Offset)))
                  (Variables|MusicEncounter|Objective|SetObjective%20Music%20Audio%20Component _audio_component))))))
        (:else)))))

(event Custom|Stop%20Music
  (bind _objective_music_audio_component (Variables|MusicEncounter|Objective|GetObjective%20Music%20Audio%20Component)
    (:then
      (Audio|Components|Audio|FadeOut _objective_music_audio_component (Variables|MusicEncounter|GetFade%20Out%20Duration) 0.0 "Linear" :!class /Script/Engine.AudioComponent)
      (EventDispatchers|BindEventtoOnAudioFinished (Variables|MusicEncounter|Objective|GetObjective%20Music%20Audio%20Component) (EventDispatchers|CreateEvent :!event Uninitialize%20Music%20Encounter)))
    (:else)))

(event Custom|On%20Owner%20End%20Play
  (params (Actor :type Actor%20Object%20Reference) (EndPlayReason :type EEndPlayReason%20Enum :default "Destroyed"))
  (CallFunction|StopMusic))

(fn AddMusicEncounterComponent
  (params (Music%20Encounter%20Component :type Music%20Encounter%20Component%20Base%20Object%20Reference%20%28by%20ref%29))
  (if (== (Variables|/MusicEncounter/Components/BP_MusicEncounterComponent.BP_MusicEncounterComponent_C|GetComponent%20Tag Music%20Encounter%20Component) (Variables|MusicEncounter|Objective|GetAttach%20Component%20Tag))
    (bind _output_get (Variables|MusicEncounter|Objective|SetObjective%20Component (Variables|GetMusic%20Encounter%20Component)))
    (|CallParentFunction _output_get)
    (return)
    (else
      (|CallParentFunction (Variables|GetMusic%20Encounter%20Component))
      (return))))
