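; Music encounter volume that plays ambient stingers at random intervals, starting when a listener enters and pausing when it exits.
; Encounter start sets a timer with a random delay between min and max and broadcasts the scheduled time.
; On expiry, unless an encounter state check blocks it, it spawns the 2D stinger MetaSound and binds audio finish to reschedule the timer.
; Uninitialize clears the timer and broadcasts cancellation.
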(event EventStartMusicEncounter
  (|CallParentFunction)
  (CallFunction|StartRetriggerTimer))

(event MusicEncounter|Volume|EventListenerEntered
  (|CallParentFunction)
  (CallFunction|StartMusicEncounter))

(event MusicEncounter|Volume|EventListenerExited
  (|CallParentFunction)
  (CallFunction|PauseMusicEncounter))

(event Custom|Start%20Retrigger%20Timer
  (bind _output_get (Variables|MusicEncounter|AmbientStingers|SetTime%20Until%20Stinger (Math|Random|RandomFloatinRange (Variables|MusicEncounter|AmbientStingers|GetRandom%20Retrigger%20Time%20Min) (Variables|MusicEncounter|AmbientStingers|GetRandom%20Retrigger%20Time%20Max))))
  (bind _returnvalue (Utilities|Time|SetTimerbyEvent (delegate Custom|Play%20Stinger) _output_get))
  (Variables|MusicEncounter|AmbientStingers|SetRetrigger%20Timer _returnvalue)
  (MusicEncounter|AmbientStingers|CallOnStingerScheduled (Variables|MusicEncounter|AmbientStingers|GetTime%20Until%20Stinger)))

(event Custom|Play%20Stinger
  (if (not (MusicEncounter|GetMusicEncounterStateValue "NewEnumerator3"))
    (bind _audio_component (MusicEncounter|SpawnMusic2D (Variables|MusicEncounter|AmbientStingers|GetStingers%20MetaSound)))
    (bind _output_get (Variables|MusicEncounter|AmbientStingers|SetAudio%20Component _audio_component))
    (Utilities|IsValid _output_get
      (:Is%20Valid
        (EventDispatchers|BindEventtoOnAudioFinished _output_get (EventDispatchers|CreateEvent :!event Custom|Start%20Retrigger%20Timer))))))

(event EventUninitializeMusicEncounter
  (|CallParentFunction)
  (Utilities|Time|ClearandInvalidateTimerbyHandle (Variables|MusicEncounter|AmbientStingers|GetAudio%20Component) (Variables|MusicEncounter|AmbientStingers|GetRetrigger%20Timer))
  (MusicEncounter|AmbientStingers|CallOnStingerCanceled))
