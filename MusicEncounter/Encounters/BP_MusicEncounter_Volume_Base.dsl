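; Base music encounter that tracks whether the audio listener is inside its AudioGameplayVolumes and relays enter and exit events.
; Initialization creates a volume delegates object, which is exposed through a getter alongside an in-volume query.
; A counter handles overlapping volumes: every entry broadcasts, but exit broadcasts only at zero or below, preventing a premature Pause.
; Uninitialization resets the count, broadcasts a final exit and clears the delegates.
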(event EventInitializeMusicEncounter
  (|CallParentFunction)
  (bind _returnvalue (Game|ConstructObjectfromClass "/MusicEncounter/DelegateObjects/BP_MusicEncounterDelegates_Volume.BP_MusicEncounterDelegates_Volume_C"))
  (Variables|MusicEncounter|Volume|SetVolume%20Delegates _returnvalue))

(event EventUninitializeMusicEncounter
  (|CallParentFunction)
  (Variables|MusicEncounter|Volume|SetVolume%20Count 0)
  (bind _volume_delegates (Variables|MusicEncounter|Volume|GetVolume%20Delegates)
    (:then
      (MusicEncounter|Volume|CallOnListenerExited _volume_delegates)
      (Variables|MusicEncounter|Volume|SetVolume%20Delegates 0))
    (:else)))

(event MusicEncounter|Volume|EventListenerEntered
  ;[1 1 1] AudioGameplayVolume events - keep track of the number of volumes the player is in to prevent overlapping volumes from calling Pause
  (Math|Integer|IncrementInt (Variables|MusicEncounter|Volume|GetVolume%20Count)
    (:%20%20
      (bind _volume_delegates (Variables|MusicEncounter|Volume|GetVolume%20Delegates)
        (:then
          (MusicEncounter|Volume|CallOnListenerEntered _volume_delegates))
        (:else)))))

(event MusicEncounter|Volume|EventListenerExited
  ;[1 1 1] AudioGameplayVolume events - keep track of the number of volumes the player is in to prevent overlapping volumes from calling Pause
  (bind _result (Math|Integer|DecrementInt (Variables|MusicEncounter|Volume|GetVolume%20Count))
    (:%20%20
      (if (<= _result 0)
        (bind _volume_delegates (Variables|MusicEncounter|Volume|GetVolume%20Delegates)
          (:then
            (MusicEncounter|Volume|CallOnListenerExited _volume_delegates))
          (:else))))))

(fn GetIsListenerinVolume
  (return (MusicEncounter|Volume|IsListenerInVolume)))

(fn GetVolumeDelegates
  (return (Variables|MusicEncounter|Volume|GetVolume%20Delegates)))

(fn IsListenerinVolume
  (return (> (Variables|MusicEncounter|Volume|GetVolume%20Count) 0)))
