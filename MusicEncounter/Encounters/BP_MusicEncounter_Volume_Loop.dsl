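; BP_MusicEncounter_Volume_Loop is a volume-based music encounter that ties looping music to the listener entering and leaving it.
; Starting the encounter spawns the Looping Music asset as 2D audio and stores the resulting audio component on the volume.
; Listener entry calls StartMusicEncounter and exit calls PauseMusicEncounter, with all three event overrides running the parent logic first.
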(event EventStartMusicEncounter
  (|CallParentFunction)
  (bind _audio_component (MusicEncounter|SpawnMusic2D (Variables|MusicEncounter|GetLooping%20Music)))
  (Variables|MusicEncounter|Volume|SetAudio%20Component _audio_component))

(event MusicEncounter|Volume|EventListenerEntered
  (|CallParentFunction)
  (CallFunction|StartMusicEncounter))

(event MusicEncounter|Volume|EventListenerExited
  (|CallParentFunction)
  (CallFunction|PauseMusicEncounter))
